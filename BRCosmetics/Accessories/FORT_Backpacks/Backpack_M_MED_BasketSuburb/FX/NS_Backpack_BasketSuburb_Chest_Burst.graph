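# One-shot Niagara burst for the BasketSuburb backpack's chest, combining warm orange puffs and glows.
# BurstPuff fires three delayed spawn groups with per-group lifetime, velocity, gravity and noise, rendered through four sprite renderers.
# ChestBurstGlows adds delayed socket-attached glow sprites with decal attributes, while ChestOuterGlow renders socket-driven glow meshes with animated material parameters.
# A time-of-day multiplier tints colors; front-end and quality-level switches adjust glow offsets, with front-end also dimming intensity.
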NiagaraSystem /BRCosmetics/Accessories/FORT_Backpacks/Backpack_M_MED_BasketSuburb/FX/NS_Backpack_BasketSuburb_Chest_Burst
dependencies {
 usedDataInterfaces [&NiagaraDataInterfaceCurve, &NiagaraDataInterfaceColorCurve, &NiagaraDataInterfaceSocketReader]
 usedDynamicInputs []
 usedModules [
  &self:SystemSpawnScript.SystemScriptSource.SystemScriptGraph.NiagaraNodeAssignment_0.SetVariables_#1
  &self:SystemSpawnScript.SystemScriptSource.SystemScriptGraph.NiagaraNodeAssignment_1.SetVariables_#2
  &self:SystemSpawnScript.SystemScriptSource.SystemScriptGraph.NiagaraNodeAssignment_2.SetVariables_#3
  &SystemState
  &self:SystemSpawnScript.SystemScriptSource.SystemScriptGraph.NiagaraNodeAssignment_3.SetVariables_#4
 ]
}
emitters {
 NE_BurstPuff {
  emitterUpdateScript [
   &EmitterState()
   &SpawnBurst_Instantaneous(Spawn Count=10, Spawn Time=$System.Smoke_Delay, Use Spawn Probability=false, UseLoopCountLimit=false)
   &SpawnBurst_Instantaneous(Spawn Count=5, Spawn Group=1, Spawn Time=$System.Smoke_Delay, Use Spawn Probability=false, UseLoopCountLimit=false) {module=SpawnBurst_Instantaneous001}
   &SpawnBurst_Instantaneous(Spawn Count=10, Spawn Group=2, Spawn Time=$System.Smoke_Delay, Use Spawn Probability=false, UseLoopCountLimit=false) {module=SpawnBurst_Instantaneous002}
  ]
  particleSpawnScript [
   /SetVariables(Particles.VisibilityTag=&SpawnGroup()) {isSetParameters=true}
   &InitializeParticle(
    Color=&RandomRangeLinearColor(Minimum=<rgba 1 0.3798 0 1>, Maximum=<rgba 1 0.5897 0 1>)
    Lifetime=&MergeHelper_FloatSwitch(
     0=&RandomRangeFloat(Minimum=1, Maximum=2)
     1=0.7
     2=0.5
     3=?
     Index=&SpawnGroup()
     Num=::NewEnumerator1
    )
    Position Offset=<0 -7.8898 14.8065>
    Sprite Size Mode=::NewEnumerator3
    Uniform Sprite Size=&RandomRangeFloat(Minimum=2, Maximum=10)
    UsePositionOffset=true
    Write Material Random=true
   )
   &AddVelocity(
    Cone Angle=90
    Cone Axis=<0 0 1>
    Module.UseConeFalloff=false
    Module.UseVelDistribution=false
    Override Seed=false
    Velocity Mode=::NewEnumerator2
    Velocity Speed=&MergeHelper_FloatSwitch(
     0=&RandomRangeFloat(Minimum=40, Maximum=100)
     1=&RandomRangeFloat(Minimum=20, Maximum=15)
     2=&RandomRangeFloat(Minimum=150, Maximum=250)
     3=?
     Index=&SpawnGroup()
     Num=::NewEnumerator1
    )
   )
   &CameraOffset(Camera Offset Amount=20)
  ]
  particleUpdateScript [
   &Drag(Drag=3, Use Linear Drag=true, Use Rotational Drag=true)
   &GravityForce(Gravity=&MergeHelper_VectorSwitch(0=<0 0 0>, 1=<0 0 0>, 2=<0 0 -400>, 3=?, Index=&SpawnGroup(), Num=::NewEnumerator1))
   &CurlNoiseForce(
    Bias Noise=false
    Noise Strength=&MergeHelper_FloatSwitch(0=10, 1=0, 2=0, 3=?, Index=&SpawnGroup(), Num=::NewEnumerator1)
    Pan Noise=false
   )
   &SolveForcesAndVelocity(Clamp Velocity=false, Limit Acceleration=false)
   &ParticleState()
   &SpriteRotationRate(Rotation Rate=&MergeHelper_FloatSwitch(
     0=&RandomRangeFloat(Minimum=-30, Maximum=30)
     1=0
     2=0
     3=?
     Index=&SpawnGroup()
     Num=::NewEnumerator1
    ))
   /SetVariables(Particles.DiamondSize:Vector2f=&Multiply_Vector2D(A=$Particles.SpriteSize, B=<1 0.25>)) {isSetParameters=true}
   /SetVariables(Particles.GlowColor=&MultiplyLinearColor(
     Linear Color=&MultiplyLinearColor(
      Linear Color=$Particles.Color
      Scale Factor=&MakeFloat4FromVectorAndFloat(
       "Vector (RGB)"=&MakeVectorFromLinearColorRGB(Color=<rgba 1 1 1 1>)
       "Float (Alpha)"=&FloatFromCurve(FloatCurve=DataInterface() {propertyValues {Curve {keys [<at 0 1>,<at 0.7 0>]}}}, Scale Curve=2)
      )
     )
     Scale Factor=&MakeVector4_FromLinearColor(LinearColor=$System.TODColorMultiplier)
    )) {isSetParameters=true}
   &DynamicMaterialParameters(
    Index 0 Param 1=&Add_Float(
     A=&FloatFromCurve(FloatCurve=DataInterface() {propertyValues {Curve {keys [{arriveTangent=2.1549, time=0, value=0.8}, <at 1 1>]}}})
     B=&Multiply_Float(A=$Particles.Initial.MaterialRandom, B=0.5)
    )
    Index 1 Param 1=1
    Index 1 Param 2=1
    Index 1 Param 3=1
    Index 1 Param 4=1
    Index 2 Param 1=&FloatFromCurve(FloatCurve=DataInterface() {propertyValues {Curve {keys [
         {arriveTangent=1.8288, interpMode=RCIM_Cubic, leaveTangent=1.8288, tangentMode=RCTM_User, time=0, value=0.3}
         {arriveTangent=-0.0004, interpMode=RCIM_Cubic, leaveTangent=-0.0004, tangentMode=RCTM_User, time=1, value=1}
        ]}}})
    Index 2 Param 2=&RandomRangeFloat(Minimum=0.5, Maximum=1)
    Index 2 Param 3=4
    Index 2 Param 4=&RandomRangeFloat(Minimum=0.8, Maximum=1.2)
    Module.Param0WriteEnabled=true
    Module.Param0WriteEnabled1=true
    Module.Param0WriteEnabled2=true
    Module.Param0WriteEnabled3=true
    Module.Param1WriteEnabled=true
    Module.Param1WriteEnabled1=true
    Module.Param1WriteEnabled2=true
    Module.Param1WriteEnabled3=true
    Module.Param2WriteEnabled=true
    Module.Param2WriteEnabled1=true
    Module.Param2WriteEnabled2=true
    Module.Param2WriteEnabled3=true
    Module.Param3WriteEnabled=true
    Module.Param3WriteEnabled1=true
    Module.Param3WriteEnabled2=true
    Module.Param3WriteEnabled3=true
    Write Parameter Index 1=true
    Write Parameter Index 2=true
    Write Parameter Index 3=true
   ) {module=DynamicMaterialParameters002}
   &FN_Sprite_Attributes(
    Color=&MultiplyLinearColor(
     Linear Color=&ColorFromCurve(ColorCurve=DataInterface() {propertyValues {
        AlphaCurve {keys [<at 0 1 RCIM_Cubic>,<at 1 0 RCIM_Cubic>]}
        BlueCurve {keys [<at 0 0.4783 RCIM_Cubic>]}
        GreenCurve {keys [<at 0 0.702 RCIM_Cubic>]}
        RedCurve {keys [<at 0 1 RCIM_Cubic>]}
       }})
     Scale Factor=&MakeVector4_FromLinearColor(LinearColor=$System.TODColorMultiplier)
    )
    Create Color Attribute=true
    Create Sprite Rotation Attribute=true
    Create SpriteSize Attribute=true
    SpriteRotation:NiagaraFloat=&Multiply_Float(A=$Particles.Age, B=&RandomRangeFloat(Minimum=90, Maximum=1))
    SpriteSize=&Vector2DFromFloat(Value=&Multiply_Float(
      A=&RandomRangeFloat(Minimum=20, Maximum=25)
      B=&FloatFromCurve(
       FloatCurve=DataInterface() {propertyValues {Curve {keys [
           {arriveTangent=1.575, interpMode=RCIM_Cubic, leaveTangent=1.575, tangentMode=RCTM_User, time=0, value=0.5}
           {interpMode=RCIM_Cubic, tangentMode=RCTM_User, time=1, value=1}
          ]}}}
       Scale Curve=1.3
      )
     ))
    Write to Module Namespace=true
   )
   &FN_Sprite_Attributes(
    Color=<rgba 2 1.0784 0.4689 1>
    Create Color Attribute=true
    Create Sprite Rotation Attribute=true
    Create SpriteSize Attribute=true
    SpriteRotation:NiagaraFloat=&Add_Float(
     A=&RandomRangeFloat(Minimum=0, Maximum=360)
     B=&Multiply_Float(A=$Particles.Age, B=&RandomRangeFloat(Minimum=90, Maximum=1))
    )
    SpriteSize=&Vector2DFromFloat(Value=&Multiply_Float(
      A=&RandomRangeFloat(Minimum=2, Maximum=3)
      B=&FloatFromCurve(FloatCurve=DataInterface() {propertyValues {Curve {keys [
           {arriveTangent=1.575, interpMode=RCIM_Cubic, leaveTangent=1.575, tangentMode=RCTM_User, time=0, value=0}
           <at 0.1 1 RCIM_Cubic>
           {interpMode=RCIM_Cubic, tangentMode=RCTM_User, time=1, value=0}
          ]}}})
     ))
    Write to Module Namespace=true
   ) {module=FN_Sprite_Attributes001}
  ]
  rendererClasses [&NiagaraSpriteRendererProperties]
  renderers [
   {rendererClass=&NiagaraSpriteRendererProperties}
   {rendererClass=&NiagaraSpriteRendererProperties, rendererIndex=1}
   {rendererClass=&NiagaraSpriteRendererProperties, rendererIndex=2}
   {rendererClass=&NiagaraSpriteRendererProperties, rendererIndex=3}
  ]
 }
 NE_ChestBurstGlows {
  emitterUpdateScript [
   &EmitterState()
   &SpawnBurst_Instantaneous(Spawn Time=$System.MainGlow_Delay, Use Spawn Probability=false, UseLoopCountLimit=false)
  ]
  particleSpawnScript [
   &InitializeParticle(
    Color=<rgba 2 2 2 1>
    Lifetime=&MaskFloatBySpawnGroup(Masked Float=0.7, Passthrough Float=2)
    Position Offset=<0 0 12>
    Sprite Size=<80 80>
    Sprite Size Mode=::NewEnumerator4
    UsePositionOffset=false
    Write Material Random=true
   )
   &CameraOffset(Camera Offset Amount=&MergeHelper_FloatSwitch(0=0, 1=0, 2=3, 3=3, 4=3, Index=$Engine.QualityLevel, Num=::NewEnumerator3))
  ]
  particleUpdateScript [
   &ParticleState()
   &ScaleSpriteSize(Uniform Curve Sprite Scale=DataInterface() {propertyValues {Curve {keys [
        {arriveTangent=2.4565, interpMode=RCIM_Cubic, leaveTangent=2.4565, tangentMode=RCTM_User, time=0, value=0}
        {interpMode=RCIM_Cubic, tangentMode=RCTM_User, time=1, value=1}
       ]}}})
   &SocketLocation(
    Orientation Sampling=::NewEnumerator1
    SamplePosOffset=true
    Sampled Position Offset=&FN_FrontEndVector3Modifier(Gameplay=<0 -12 19>, Front End=<0 -10 19>)
    Socket Index=1
    Socket Reader=$System.SocketReader
    Socket Sampling Mode=::NewEnumerator3
    VelocitySpeedLimit=false
   )
   &ScaleColor(
    Scale Alpha=&FloatFromCurve(FloatCurve=DataInterface() {propertyValues {Curve {keys [
         {interpMode=RCIM_Cubic, tangentMode=RCTM_User, time=0, value=1}
         {arriveTangent=-2.5387, interpMode=RCIM_Cubic, leaveTangent=-2.5387, tangentMode=RCTM_User, time=1, value=0}
        ]}}})
    Scale RGB=&MakeVectorFromLinearColorRGB(Color=&ColorFromCurve(ColorCurve=DataInterface() {propertyValues {
        AlphaCurve {keys [<at 0 1 RCIM_Cubic>,<at 1 0 RCIM_Cubic>]}
        BlueCurve {keys [<at 0 0 RCIM_Cubic>,<at 1 0 RCIM_Cubic>]}
        GreenCurve {keys [<at 0 2.5226 RCIM_Cubic>,<at 1 0.3066 RCIM_Cubic>]}
        RedCurve {keys [<at 0 5 RCIM_Cubic>,<at 1 1 RCIM_Cubic>]}
       }}))
    ScaleA=true
    ScaleRGB=true
   )
   /SetVariables(Particles.GlowColor=&MultiplyLinearColor(
     Linear Color=&MultiplyLinearColor(
      Linear Color=$Particles.Color
      Scale Factor=&MakeFloat4FromVectorAndFloat(
       "Vector (RGB)"=&MakeVectorFromLinearColorRGB(Color=<rgba 1 1 1 1>)
       "Float (Alpha)"=&FloatFromCurve(FloatCurve=DataInterface() {propertyValues {Curve {keys [<at 0 1>,<at 0.7 0>]}}}, Scale Curve=2)
      )
     )
     Scale Factor=&MakeVector4_FromLinearColor(LinearColor=$System.TODColorMultiplier)
    )) {isSetParameters=true}
   /SetVariables(Particles.GroundSpotSize:Vector2f=&Multiply_Vector2D(A=$Particles.SpriteSize, B=<0.25 0.5>)) {isSetParameters=true}
   /SetVariables(Particles.GlowDecalSize:Vector2f=&Multiply_Vector2D_ByFloat(Float=50, Vector2D=<1 1>)) {isSetParameters=true}
   /SetVariables(Particles.DecalDynMat:Vector4f=<3 0 0 0>) {isSetParameters=true}
   /SetVariables(Particles.DecalColor:LinearColor=&MultiplyLinearColor(
     Linear Color=&MultiplyLinearColor(
      Linear Color=&MakeLinearColorFromVectorAndFloat(
       "Vector (RGB)"=&VectorFromFloat(Value=&FN_FrontEndFloatModifier(Gameplay=0.7, Front End=0.5))
       "Float (Alpha)"=&FloatFromCurve(
        FloatCurve=DataInterface() {propertyValues {Curve {
           defaultValue=1
           keys [
            {interpMode=RCIM_Cubic, tangentMode=RCTM_User, time=0, value=0}
            {arriveTangent=-1.5896, interpMode=RCIM_Cubic, leaveTangent=-1.5896, tangentMode=RCTM_User, time=0.1, value=0.9999}
            {interpMode=RCIM_Cubic, tangentMode=RCTM_User, time=1, value=0}
           ]
          }}}
        Scale Curve=&FN_FrontEndFloatModifier(Gameplay=0.7, Front End=0.5)
       )
      )
      Scale Factor=&MakeVector4_FromLinearColor(LinearColor=$Particles.Color)
     )
     Scale Factor=&MakeVector4_FromLinearColor(LinearColor=$System.TODColorMultiplier)
    )) {isSetParameters=true}
   /SetVariables(Particles.DecalCameraOffset=25) {isSetParameters=true}
   &FN_Sprite_Attributes(
    Color=&MultiplyLinearColor(
     Linear Color=&MultiplyLinearColor(Linear Color=$Particles.Color, Scale Factor=<1 1 1 0.3>)
     Scale Factor=&MakeVector4_FromLinearColor(LinearColor=$System.TODColorMultiplier)
    )
    Create Color Attribute=true
    Create SpriteSize Attribute=true
    SpriteSize=&Multiply_Vector2D_ByFloat(Float=0.5, Vector2D=$Particles.SpriteSize)
    Write to Module Namespace=true
   )
   /SetVariables(Particles.DecalPosition:NiagaraPosition=&AddVectorToPosition(Position=$Particles.Position, Vector=<0 0 0>)) {isSetParameters=true}
  ]
  rendererClasses [&NiagaraSpriteRendererProperties]
  renderers [
   {rendererClass=&NiagaraSpriteRendererProperties}
   {rendererClass=&NiagaraSpriteRendererProperties, rendererIndex=1}
   {rendererClass=&NiagaraSpriteRendererProperties, rendererIndex=2}
  ]
 }
 NE_ChestOuterGlow {
  emitterUpdateScript [
   &EmitterState()
   &SpawnBurst_Instantaneous(Use Spawn Probability=false, UseLoopCountLimit=false)
   &SpawnBurst_Instantaneous(Spawn Group=1, Use Spawn Probability=false, UseLoopCountLimit=false) {module=SpawnBurst_Instantaneous001}
   /SetVariables(Emitter.PanSpeed=0.25) {isSetParameters=true}
  ]
  particleSpawnScript [
   /SetVariables(Particles.VisibilityTag=&SpawnGroup()) {isSetParameters=true}
   &InitializeParticle(
    Color Mode=::NewEnumerator0
    Lifetime=1.2
    Mesh Scale=<0.1 0.1 1>
    Mesh Scale Mode=::NewEnumerator4
    Sprite Size=<40 40>
    Sprite Size Mode=::NewEnumerator4
    UsePositionOffset=false
    Write Material Random=true
   )
   &InitialMeshOrientation(
    Mesh Axis To Orient=<0 0 1>
    Mesh Orientation Mode=::NewEnumerator2
    Orientation Vector=<0 0 1>
    Rotation=<0 0 0.25>
    Use Rotation Vector=true
   )
   /SetVariables() {isSetParameters=true}
   &CameraOffset(Camera Offset Amount=&FN_FrontEndFloatModifier(
     Gameplay=&MergeHelper_FloatSwitch(0=-15, 1=-15, 2=-5, 3=-5, 4=-5, Index=$Engine.QualityLevel, Num=::NewEnumerator3)
     Front End=0
    ))
  ]
  particleUpdateScript [
   &ParticleState()
   &SocketLocation(
    Orientation Sampling=::NewEnumerator1
    Position Sampling=::NewEnumerator1
    SamplePosOffset=false
    Socket Reader=$System.SocketReader
    Socket Sampling Mode=::NewEnumerator3
    VelocitySpeedLimit=false
   )
   /SetVariables(Particles.Position=&MaskVectorBySpawnGroup(Masked Vector=$Output.SocketLocation.SampledPosition, Passthrough Vector=<0 -10 14>)) {isSetParameters=true}
   &DynamicMaterialParameters(
    Index 0 Param 1=&Multiply_Float(
     A=0.12
     B=&FloatFromCurve(
      FloatCurve=DataInterface() {propertyValues {Curve {
         defaultValue=1
         keys [
          {arriveTangent=13.7172, interpMode=RCIM_Cubic, leaveTangent=13.7172, tangentMode=RCTM_User, time=0.03, value=0}
          {arriveTangent=45.0655, leaveTangent=0.6956, tangentMode=RCTM_User, time=0.085, value=1.1}
          {arriveTangent=-4.3367, leaveTangent=-4.3367, time=0.25, value=1}
         ]
        }}}
      Scale Curve=&Multiply_Float(A=0.7227, B=&MakeFloatFromVector(Vector=$Output.SocketLocation.SampledScale, Channel=::NewEnumerator2))
     )
    )
    Index 0 Param 2=0.1
    Index 0 Param 3=&Lerp_Float(
     A=0
     B=0.01
     Alpha=&FloatFromCurve(
      FloatCurve=DataInterface() {propertyValues {Curve {
         defaultValue=1
         keys [
          {arriveTangent=13.7172, interpMode=RCIM_Cubic, leaveTangent=13.7172, tangentMode=RCTM_User, time=0.03, value=0}
          {arriveTangent=45.0655, leaveTangent=0.6956, tangentMode=RCTM_User, time=0.085, value=1.1}
          {arriveTangent=-4.3367, leaveTangent=-4.3367, time=0.25, value=1}
         ]
        }}}
      Scale Curve=&Multiply_Float(A=0.7227, B=&MakeFloatFromVector(Vector=$Output.SocketLocation.SampledScale, Channel=::NewEnumerator2))
     )
    )
    Index 0 Param 4=50
    Module.Index0Float0Write=true
    Module.Index0Float1Write=true
    Module.Index0Float2Write=true
    Module.Index0Float3Write=true
   )
   &Color(Color=&MultiplyLinearColor(
     Linear Color=&MakeLinearColorFromVectorAndFloat(
      "Vector (RGB)"=&MakeVectorFromLinearColorRGB(Color=<rgba 3 1.5965 0 1>)
      "Float (Alpha)"=&FloatFromCurve(
       FloatCurve=DataInterface() {propertyValues {Curve {keys [
           <at 0 0 RCIM_Cubic>
           {arriveTangent=-2.957, interpMode=RCIM_Cubic, leaveTangent=-2.957, tangentMode=RCTM_User, time=0.1, value=1}
           {arriveTangent=-0.5684, interpMode=RCIM_Cubic, leaveTangent=-0.5684, tangentMode=RCTM_User, time=1, value=0}
          ]}}}
       Scale Curve=&FN_FrontEndFloatModifier(Gameplay=1, Front End=0.75)
      )
     )
     Scale Factor=&MakeVector4_FromLinearColor(LinearColor=$System.TODColorMultiplier)
    ))
   &FN_Mesh_Attributes(
    Color=&MultiplyLinearColor(
     Linear Color=&MakeLinearColorFromVectorAndFloat(
      "Vector (RGB)"=&MakeVectorFromLinearColorRGB(Color=<rgba 3 1.5965 0 1>)
      "Float (Alpha)"=&FloatFromCurve(
       FloatCurve=DataInterface() {propertyValues {Curve {keys [
           <at 0 0 RCIM_Cubic>
           {arriveTangent=-2.452, interpMode=RCIM_Cubic, leaveTangent=-2.452, tangentMode=RCTM_User, time=0.1, value=1}
           {arriveTangent=-0.934, interpMode=RCIM_Cubic, leaveTangent=-0.934, tangentMode=RCTM_User, time=1, value=0}
          ]}}}
       Scale Curve=&FN_FrontEndFloatModifier(Gameplay=1, Front End=0.75)
      )
     )
     Scale Factor=&MakeVector4_FromLinearColor(LinearColor=$System.TODColorMultiplier)
    )
    Create Color Attribute=true
    Create Orientation Attribute=true
    Create Scale Attribute=true
    Orientation:Quat4f=&MultiplyQuaternion(
     Quaternion A=$Output.SocketLocation.SampledRotation
     Quaternion B=&MakeQuatFromAxisAngle(Axis=<1 0 0>, AngleInDegrees=90)
    )
    Scale=&Multiply_Vector(
     A=&Multiply_Vector(A=<1 1.1866 0.6093>, B=&VectorFromFloat(Value=0.1893))
     B=$Output.SocketLocation.SampledScale
    )
    Write to Module Namespace=true
   ) {module=FN_Mesh_Attributes001}
   /SetVariables(Particles.ConePos:NiagaraPosition=&AddVectorToPosition(Position=$Particles.Position, Vector=<0 0 -5.1922>)) {isSetParameters=true}
  ]
  rendererClasses [&NiagaraMeshRendererProperties]
  renderers [
   {rendererClass=&NiagaraMeshRendererProperties, rendererIndex=0}
   {rendererClass=&NiagaraMeshRendererProperties, rendererIndex=1}
  ]
 }
}
summary {
 systemName=NS_Backpack_BasketSuburb_Chest_Burst
 emitters [
  {emitterName=NE_ChestBurstGlows, rendererClasses [&NiagaraSpriteRendererProperties]}
  {emitterName=NE_BurstPuff, rendererClasses [&NiagaraSpriteRendererProperties]}
  {emitterName=NE_ChestOuterGlow, rendererClasses [&NiagaraMeshRendererProperties]}
 ]
 userVariables [
  User.bIsFrontEnd {type=NiagaraBool}
  User.bIsFrontEndPreview {type=NiagaraBool}
 ]
}
