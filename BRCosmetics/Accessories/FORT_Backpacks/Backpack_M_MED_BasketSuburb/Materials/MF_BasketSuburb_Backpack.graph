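# Material function for the BasketSuburb backpack that writes an animated, expressive face with tears into the emissive channel of incoming material attributes.
# Blink, Happy and Anguish pose parameters pick frames from a channel-packed 2x2 flipbook.
# Scrolling noise distorts a second flipbook sample for tears and highlights, with strength varying by pose and camera depth.
# SDF box and circle masks confine the layers, and a shading-gradient input tints the tear color.
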MaterialFunction /BRCosmetics/Accessories/FORT_Backpacks/Backpack_M_MED_BasketSuburb/Materials/MF_BasketSuburb_Backpack
graph {nodes [
  Constant()
  n1 = GetMaterialAttributes(None=FunctionInput() {inputName=Result, inputType=FunctionInput_MaterialAttributes}) {attributeGetTypes [#1]}
  n2 = VectorParameter() {group=BasketSuburb - Backpack, parameterName=Tear Color, defaultValue=<rgba 0.4098 0.9461 1 1>}
  n3 = Multiply(A=@n2.RGB) {constB=0.8}
  n4 = Saturate(None=Desaturation(None=FunctionInput() {inputName=Shading_Gradient, inputType=FunctionInput_Scalar, sortPriority=10}))
  n5 = Round(None=ScalarParameter() {group=BasketSuburb - Backpack, parameterName=Blink Pose})
  n6 = ScalarParameter() {group=BasketSuburb - Backpack, parameterName=Happy Pose}
  n7 = LinearInterpolate(
   A=LinearInterpolate(A=LinearInterpolate(Alpha=@n5) {constA=0.75, constB=0.25}, Alpha=Round(None=@n6)) {constB=1.5}
   Alpha=ScalarParameter() {group=BasketSuburb - Backpack, parameterName=Anguish Pose}
  )
  n8 = Constant() {r=2}
  n9 = TextureObject() {
   samplerType=SAMPLERTYPE_Masks
   texture=/BRCosmetics/Accessories/FORT_Backpacks/Backpack_M_MED_BasketSuburb/Textures/T_BasketSuburb_Flipbook_01
  }
  n10 = TextureCoordinate()
  n11 = VectorParameter() {group=BasketSuburb - Backpack, parameterName=Expression Offset, defaultValue=<rgba 0.4312 0.1434 0 0>}
  n12 = MaterialFunctionCall(X=@n11.R, Y=@n11.R) {materialFunction=&MakeFloat2}
  n13 = VectorParameter() {group=BasketSuburb - Backpack, parameterName=Expression Tiling, defaultValue=<rgba 0.1 0.2 0 0>}
  n14 = MaterialFunctionCall(
   UVs=MaterialFunctionCall(
    UVs=Add(A=@n10, B=@n12.Result)
    Texture Scale=MaterialFunctionCall(X=@n13.R, Y=@n13.R) {materialFunction=&MakeFloat2}.Result
   ) {materialFunction=&ScaleUVsByCenter}.UVs
   "Rotation Angle (0-1)"=ScalarParameter() {defaultValue=0.25, group=BasketSuburb - Backpack, parameterName=Expression Rotation}
  ) {materialFunction=&CustomRotator}
  n15 = VectorParameter() {group=BasketSuburb - Backpack, parameterName=Noise XY, defaultValue=<rgba 15 10 0 0>}
  n16 = MaterialFunctionCall(X=@n15.R, Y=@n15.R) {materialFunction=&MakeFloat2}
  n17 = Multiply(A=@n16.Result, B=ScalarParameter() {defaultValue=2, group=BasketSuburb - Backpack, parameterName=Highlight Wobble Tiling})
  n18 = NamedRerouteUsage()
  n19 = VectorParameter() {group=BasketSuburb - Backpack, parameterName=Noise XY - Anguish, defaultValue=<rgba 30 25 0 0>}
  n20 = Saturate(None=ScalarParameter() {group=BasketSuburb - Backpack, parameterName=Anguish Pose})
  n21 = Multiply(
   A=TextureCoordinate()
   B=LinearInterpolate(
    A=LinearInterpolate(A=@n16.Result, B=@n17, Alpha=@n18)
    B=MaterialFunctionCall(X=@n19.R, Y=@n19.R) {materialFunction=&MakeFloat2}.Result
    Alpha=@n20
   )
  )
  n22 = VectorParameter() {group=BasketSuburb - Backpack, parameterName=Noise Speed, defaultValue=<rgba -0.3 0.1 0 0>}
  n23 = Multiply(
   A=MaterialFunctionCall(X=@n22.R, Y=@n22.R) {materialFunction=&MakeFloat2}.Result
   B=Time()
  )
  n24 = LinearInterpolate(
   A=ScalarParameter() {defaultValue=0.05, group=BasketSuburb - Backpack, parameterName=Noise Distort}
   B=ScalarParameter() {group=BasketSuburb - Backpack, parameterName=Anguish - Distort Noise}
   Alpha=@n20
  ) {constA=0.02}
  n25 = ScalarParameter() {defaultValue=0.06, group=BasketSuburb - Backpack, parameterName=Blink - Distort}
  n26 = Saturate(None=ScalarParameter() {group=BasketSuburb - Backpack, parameterName=Blink Pose})
  n27 = ScalarParameter() {defaultValue=100, group=BasketSuburb - Backpack, parameterName=Highlight Wobble Distortion Fade}
  n28 = LinearInterpolate(
   A=ScalarParameter() {defaultValue=0.03, group=BasketSuburb - Backpack, parameterName=Highlight Wobble Distortion}
   B=ScalarParameter() {defaultValue=0.05, group=BasketSuburb - Backpack, parameterName=Highlight Wobble Distortion Far}
   Alpha=Saturate(None=MaterialFunctionCall(Fade Length=@n27) {materialFunction=&CameraDepthFade}.Result)
  )
  n29 = Round(None=ScalarParameter() {group=BasketSuburb - Backpack, parameterName=Happy Pose})
  n30 = Multiply(
   A=TextureSample(UVs=Add(A=@n21, B=@n23)) {samplerType=SAMPLERTYPE_Grayscale, texture=&LowResBlurredNoise}.R
   B=LinearInterpolate(A=LinearInterpolate(A=@n24, B=@n25, Alpha=@n26) {constB=0.02}, B=LinearInterpolate(A=@n28, Alpha=@n29) {constB=0}, Alpha=NamedRerouteUsage())
  )
  n31 = ComponentMask(None=VectorParameter() {group=BasketSuburb - Backpack, parameterName=Tear Position, defaultValue=<rgba -0.027 -0.0054 0 0>}.RGB) {g=true, r=true}
  n32 = LinearInterpolate(A=@n31, B=Divide(A=@n31), Alpha=ScalarParameter() {group=BasketSuburb - Backpack, parameterName=Blink Pose}) {constB=0}
  n33 = OneMinus(None=NamedRerouteUsage())
  n34 = MaterialFunctionCall!(
   "Animation  Phase (0-1)"=@n7
   Number of Rows=@n8
   Number of Columns=@n8
   Texture=@n9
   UVs=Add(A=Add(A=@n14.Rotated Values, B=@n30), B=Multiply(A=@n32, B=@n33)) {constB=0}
   MipBias/Level
   "Use Mip Bias (T) Level (F)"
   "Clamp Anim (see tooltip)"
  ) {materialFunction=&/MF_FlipBook_PackedChannels}
  n35 = Step(X=@n34.G) {constX=0.01, constY=0.4}
  n36 = VectorParameter() {group=BasketSuburb - Backpack, parameterName=Face Mask - Position, defaultValue=<rgba 0.2984 -0.3659 0 0>}
  n37 = Add(
   A=TextureCoordinate()
   B=MaterialFunctionCall(X=@n36.R, Y=@n36.R) {materialFunction=&MakeFloat2}.Result
  )
  n38 = VectorParameter() {group=BasketSuburb - Backpack, parameterName=Size, defaultValue=<rgba 0.0019 0.0477 0 0>}
  n39 = Multiply(A=@n38.R, B=ScalarParameter() {defaultValue=0.832, group=BasketSuburb - Backpack, parameterName=Size - Blink}) {constB=20}
  n40 = ScalarParameter() {group=BasketSuburb - Backpack, parameterName=Blink Pose}
  n41 = MaterialFunctionCall!(
   UVs=@n37
   Size=MaterialFunctionCall(X=LinearInterpolate(A=@n38.R, B=@n39, Alpha=@n40), Y=@n38.G) {materialFunction=&MakeFloat2}.Result
   CornerRoundness=ScalarParameter() {defaultValue=1, group=BasketSuburb - Backpack, parameterName=Roundness}
   KeepAspectRatio
  ) {materialFunction=/FortUILibrary/MaterialFunctions/Shapes/SDFs/MF_SDF_Box}
  n42 = LinearInterpolate(
   A=@n1.EmissiveColor
   B=Multiply(A=LinearInterpolate(A=@n3, B=@n2.RGB, Alpha=@n4), B=@n34.G)
   Alpha=Multiply(A=Saturate(None=@n35), B=Step(Y=@n41.SDF) {constX=0.5})
  )
  n43 = MaterialFunctionCall("Animation  Phase (0-1)"=@n7, Number of Rows=@n8, Number of Columns=@n8, Texture=@n9, UVs=@n14.Rotated Values) {materialFunction=&MF_FlipBook_PackedChannels}
  n44 = Step(X=@n43.B) {constX=0.01, constY=0.8}
  n45 = MaterialFunctionCall!(UVs=@n37, Radius=ScalarParameter() {defaultValue=0.0945, group=BasketSuburb - Backpack, parameterName=Face Mask - Radius}) {materialFunction=/FortUILibrary/MaterialFunctions/Shapes/SDFs/MF_SDF_Circle}
  n46 = Saturate(None=Multiply(A=@n44, B=Step(Y=@n45.SDF) {constX=0.1}))
  n47 = Max(A=@n43.R, B=@n34.A)
  FunctionOutput(None=SetMaterialAttributes!(MaterialAttributes=@n1.MaterialAttributes, EmissiveColor=LinearInterpolate(A=@n42, B=@n47, Alpha=@n46)) {attributeSetTypes [#1]}) {outputName=SetMaterialAttributes}
  n48 = OneMinus(None=@n35)
  Multiply(A=Saturate(None=Subtract(A=@n35, B=@n48)), B=@n34.G)
  NamedRerouteDeclaration(None=@n44) {name=Eye Mask, nodeColor=<rgba 0 0.5547 1 1>}
  Step(X=@n43.B) {constX=0.01, constY=0.01}
 ]}
