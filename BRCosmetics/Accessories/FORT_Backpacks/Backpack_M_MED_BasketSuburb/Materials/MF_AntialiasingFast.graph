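# MF_AntialiasingFast is a material function in the DelMarUI library category that turns a scalar SDF into antialiased masks.
# It divides the SDF by the length of its screen-space derivatives plus an optional SDF Width for gradients, then saturates for Mask.
# Opacity Mask saturates the width's sign minus that ratio, meant to avoid antialiasing pixel bleed when used for opacity.
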MaterialFunction /BRCosmetics/Accessories/FORT_Backpacks/Backpack_M_MED_BasketSuburb/Materials/MF_AntialiasingFast
graph {
 properties {bExposeToLibrary=true, description=Derivative based antialiasing., libraryCategoriesText [DelMarUI]}
 nodes [
  n1 = FunctionInput() {
   bUsePreviewValueAsDefault=true
   description=SDF must be calculated correctly or aliasing artifacts will be visible.
   inputName=SDF
   inputType=FunctionInput_Scalar
   previewValue=<1 0 0 1>
  }
  n2 = AppendVector(A=DDX(Value=@n1), B=DDY(Value=@n1))
  n3 = FunctionInput() {
   bUsePreviewValueAsDefault=true
   description=Useful if gradients are needed.
   inputName=SDF Width
   inputType=FunctionInput_Scalar
   sortPriority=1
  }
  n4 = Divide(A=@n1, B=Add(A=Length(None=@n2) {is=class MaterialExpressionLength}, B=@n3))
  FunctionOutput(None=Saturate(None=@n4)) {outputName=Mask}
  n5 = Sign(None=@n3)
  FunctionOutput(None=Saturate(None=Subtract(A=@n5, B=@n4))) {description=For use if mask will be used for opacity. This fixes antialiasing pixel bleed., outputName=Opacity Mask, sortPriority=1}
 ]
}
