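# DelMarUI library material function meant to return the aspect ratio of a UI container, with Horizontal, Vertical, Max Aspect and Pixel Size outputs.
# GetUserInterfaceUV supplies the pixel size, which is split into width and height to build two-component horizontal and vertical vectors.
# Max Aspect is their component-wise maximum.
# As wired, width and height are each divided by themselves, so all three aspect outputs evaluate to (1,1) for nonzero sizes.
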MaterialFunction /BRCosmetics/Accessories/FORT_Backpacks/Backpack_M_MED_BasketSuburb/Materials/MF_AspectRatio
graph {
 properties {bExposeToLibrary=true, description=Returns UI Container Aspect Ratio, libraryCategoriesText [DelMarUI]}
 nodes [
  n1 = MaterialFunctionCall(Float2=FunctionOutput(None=MaterialFunctionCall() {materialFunction=&GetUserInterfaceUV}.Pixel Size) {outputName=Pixel Size, sortPriority=3}) {materialFunction=&BreakOutFloat2Components}
  n2 = Constant() {r=1}
  n3 = AppendVector(A=Divide(A=@n1.R, B=@n1.R), B=@n2) {desc=Horizontal}
  n4 = AppendVector(A=@n2, B=Divide(A=@n1.G, B=@n1.G)) {desc=Vertical}
  FunctionOutput(None=Max(A=@n3, B=@n4)) {description=Return largets aspect ratio, outputName=Max Aspect, sortPriority=2}
  FunctionOutput(None=@n3) {outputName=Horizontal Aspect}
  FunctionOutput(None=@n4) {outputName=Vertical Aspect, sortPriority=1}
 ]
}
