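# Material function for the BasketSuburb backpack cosmetic that builds a signed distance field and masks for an arbitrary four-cornered polygon, exposed under SDF.
# UVs and corners are scaled by MF_AspectRatio, then a Custom HLSL node computes edge distance with crossing-based sign, doubled in magnitude.
# Inset offsets shift the field, which MF_AntialiasingFast turns into base and inner masks.
# Outputs are the SDF, both masks, and their difference as an outline.
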MaterialFunction /BRCosmetics/Accessories/FORT_Backpacks/Backpack_M_MED_BasketSuburb/Materials/MF_SDPoly
graph {
 properties {bExposeToLibrary=true, description=Signed distance field for a 4-sided polygon., libraryCategoriesText [SDF]}
 nodes [
  n1 = FunctionInput() {
   bUsePreviewValueAsDefault=true
   description=Inset or scale down base mask.
   inputName=InsetBaseMask
   inputType=FunctionInput_Scalar
   sortPriority=5
   previewValue=<0 0 0 0>
  }
  n2 = MaterialFunctionCall() {materialFunction=&/MF_AspectRatio}
  n3 = FunctionInput(Preview=TextureCoordinate()) {
   bUsePreviewValueAsDefault=true
   description="Coordinates of the distance field (0,0 is origin)."
   inputName=UVs
   inputType=FunctionInput_Vector2
  }
  MaterialExpressionMultiply_0 = Multiply(A=@n2.Max Aspect, B=@n3)
  n4 = FunctionInput() {
   bUsePreviewValueAsDefault=true
   description=The coordinates of the first vertex.
   inputName=TopLeft
   inputType=FunctionInput_Vector2
   sortPriority=1
   previewValue=<0.1 0.1 0 0>
  }
  MaterialExpressionMultiply_1 = Multiply(A=@n2.Max Aspect, B=@n4)
  n5 = FunctionInput() {
   bUsePreviewValueAsDefault=true
   description=The coordinates of the second vertex.
   inputName=TopRight
   inputType=FunctionInput_Vector2
   sortPriority=2
   previewValue=<0.9 0.1 0 0>
  }
  MaterialExpressionMultiply_2 = Multiply(A=@n2.Max Aspect, B=@n5)
  n6 = FunctionInput() {
   bUsePreviewValueAsDefault=true
   description=The coordinates of the third vertex.
   inputName=BottomRight
   inputType=FunctionInput_Vector2
   sortPriority=3
   previewValue=<0.9 0.9 0 0>
  }
  MaterialExpressionMultiply_3 = Multiply(A=@n2.Max Aspect, B=@n6)
  n7 = FunctionInput() {
   bUsePreviewValueAsDefault=true
   description=The coordinates of the final vertex.
   inputName=BottomLeft
   inputType=FunctionInput_Vector2
   sortPriority=4
   previewValue=<0.1 0.9 0 0>
  }
  MaterialExpressionMultiply_4 = Multiply(A=@n2.Max Aspect, B=@n7)
  n8 = Custom!(
   p=@MaterialExpressionMultiply_0
   v1=@MaterialExpressionMultiply_1
   v2=@MaterialExpressionMultiply_2
   v3=@MaterialExpressionMultiply_3
   v4=@MaterialExpressionMultiply_4
  ) {
   code="float2 v[] = {v1,v2,v3,v4};\r\nconst int N = 4;\r\n\r\nfloat d = dot(p-v[0],p-v[0]);\r\nfloat s = 2.0;\r\nfor( int i=0, j=N-1; i<N; j=i, i++ )\r\n{\r\n    float2 e = v[j] - v[i];\r\n    float2 w = p - v[i];\r\n    float2 b = w - e*clamp( dot(w,e)/dot(e,e), 0.0, 1.0 );\r\n    d = min( d, dot(b,b) );\r\n    float3 c = float3(p.y>=v[i].y,p.y<v[j].y,e.x*w.y>e.y*w.x);\r\n    if( all(c) || all(!c) ) s = -s;\r\n}\r\nreturn s*sqrt(d);"
   description=PolySDF
   outputType=CMOT_Float1
   inputs [
    {
     inputName=p
     input {expression=&self:MaterialExpressionMultiply_0}
    }
    {
     inputName=v1
     input {expression=&self:MaterialExpressionMultiply_1}
    }
    {
     inputName=v2
     input {expression=&self:MaterialExpressionMultiply_2}
    }
    {
     inputName=v3
     input {expression=&self:MaterialExpressionMultiply_3}
    }
    {
     inputName=v4
     input {expression=&self:MaterialExpressionMultiply_4}
    }
   ]
  }
  n9 = FunctionOutput(None=Add(A=@n1, B=@n8)) {outputName=SDF, sortPriority=10}
  n10 = FunctionOutput(None=OneMinus(None=MaterialFunctionCall!(SDF=@n9, SDF Width) {materialFunction=&/MF_AntialiasingFast}.Mask)) {description=Base poly mask., outputName=BaseMask}
  n11 = FunctionInput() {
   bUsePreviewValueAsDefault=true
   description="This insets the inner pixed of the base mask. It can also be used as outline thickness. "
   inputName=InsetInnerMask
   inputType=FunctionInput_Scalar
   sortPriority=6
   previewValue=<0.1 0 0 0>
  }
  n12 = Add(A=Add(A=@n1, B=@n11), B=@n8)
  n13 = FunctionOutput(None=OneMinus(None=MaterialFunctionCall(SDF=@n12) {materialFunction=&MF_AntialiasingFast}.Mask)) {description="Inner mask, must have a inner inset greater than 0.", outputName=InnerMask, sortPriority=1}
  FunctionOutput(None=Saturate(None=Subtract(A=@n10, B=@n13))) {description=The difference between the base and inner mask. Creating an outline., outputName=Outline, sortPriority=2}
 ]
}
