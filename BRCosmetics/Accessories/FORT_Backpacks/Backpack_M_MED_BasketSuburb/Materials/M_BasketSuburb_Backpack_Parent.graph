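# Parent Cloth-shaded material for the BasketSuburb backpack cosmetic, built on the 2023 Anime Parent character shader.
# The shader's Pre, Calculations and Post functions run in sequence, then MF_BasketSuburb_Backpack modifies the result using the shading gradient.
# A separate branch offsets, scales and rotates UVs to build two MF_SDPoly quad masks, one mirrored, merged with Max.
# That mask ends in a Subtract with no visible consumer.
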Material /BRCosmetics/Accessories/FORT_Backpacks/Backpack_M_MED_BasketSuburb/Materials/M_BasketSuburb_Backpack_Parent
graph {
 properties {
  bAutomaticallySetUsageInEditor=false
  bUseMaterialAttributes=true
  bUsedWithMeshDeformer=true
  bUsedWithMorphTargets=true
  bUsedWithSkeletalMesh=true
  bUsedWithStaticLighting=true
  pixelDepthOffsetMode=PDOM_Legacy
  shadingModel=MSM_Cloth
  usedShadingModels=Cloth
 }
 nodes [
  n1 = MaterialFunctionCall!(CustomUV) {materialFunction=/Game/Global/Materials/Character_Parent_2023/Anime_Parent_2023/MaterialFunctions/MFC_AnimeShader_2023_Pre}
  n2 = MaterialFunctionCall!(CustomLightVector, OverrideNormal_WS, STT=@n1.STT, Normal_Map=@n1.STT, DFL=@n1.STT, FX=@n1.STT) {materialFunction=/Game/Global/Materials/Character_Parent_2023/Anime_Parent_2023/MaterialFunctions/MFC_AnimeShader_2023_Calculations}
  MaterialFunctionCall!(
   Result=MaterialFunctionCall!(
    Shading_Gradient=@n2.Shading_Gradient
    Threshold_Gradient=@n2.Shading_Gradient
    Specular_Gradient=@n2.Shading_Gradient
    Sheen_Mask=@n2.Shading_Gradient
    Thickness=@n2.Shading_Gradient
    DFL_Mask=@n2.Shading_Gradient
    FX=@n2.Shading_Gradient
    Metallic_Gradient=@n2.Shading_Gradient
    RimLight_Gradient=@n2.Shading_Gradient
    NormalBuffer=@n2.Shading_Gradient
    --------
    CINE_LightSources=@n2.Shading_Gradient
    CINE_LightColor=@n2.Shading_Gradient
    ---------
    Color_Lit=@n1.Color_Lit
    Color_Shaded=@n1.Color_Lit
    SRM=@n1.Color_Lit
    Emissive=@n1.Color_Lit
    DFL_Color=@n1.Color_Lit
    Specular_Color=@n1.Color_Lit
    RimLight_Color=@n1.Color_Lit
   ) {materialFunction=/Game/Global/Materials/Character_Parent_2023/Anime_Parent_2023/MaterialFunctions/MFC_AnimeShader_2023_Post}.Result
   Shading_Gradient=@n2.Shading_Gradient
  ) {materialFunction=&/MF_BasketSuburb_Backpack}
  n3 = Add(A=Multiply(B=ScalarParameter() {defaultValue=0.07, parameterName=Cut - Distort Intensity}) {constB=0.01}, B=TextureCoordinate())
  n4 = VectorParameter() {parameterName=SDF Pos, defaultValue=<rgba -0.3396 0.2592 0 0>}
  n5 = MaterialFunctionCall(X=@n4.G, Y=@n4.G) {materialFunction=&MakeFloat2}
  n6 = MaterialFunctionCall(
   UVs=MaterialFunctionCall(UVs=Add(A=@n3, B=@n5.Result), Texture Scale=ScalarParameter() {defaultValue=0.0761, parameterName=Scale SDF}) {materialFunction=&ScaleUVsByCenter}.UVs
   "Rotation Angle (0-1)"=ScalarParameter() {defaultValue=0.25, parameterName="Rotation Angle (0-1) (S)"}
  ) {materialFunction=&CustomRotator}
  n7 = VectorParameter() {defaultValue=<rgba 0.636 0.1 0 0>}
  n8 = VectorParameter() {parameterName=Param_1, defaultValue=<rgba 0.6392 0.1 0 0>}
  n9 = VectorParameter() {parameterName=Param_2, defaultValue=<rgba 0.634 0.9 0 0>}
  n10 = VectorParameter() {parameterName=Param_3, defaultValue=<rgba 0.562 0.9 0 0>}
  n11 = MaterialFunctionCall(Float2=@n6.Rotated Values) {materialFunction=&BreakOutFloat2Components}
  n12 = VectorParameter() {parameterName=SDF Offset, defaultValue=<rgba 0.0584 0.1144 0 0>}
  n13 = Add(
   A=MaterialFunctionCall(X=OneMinus(None=@n11.R), Y=@n11.G) {materialFunction=&MakeFloat2}.Result
   B=MaterialFunctionCall(X=@n12.R, Y=@n12.R) {materialFunction=&MakeFloat2}.Result
  )
  Subtract(B=Max(
    A=MaterialFunctionCall!(UVs=@n6.Rotated Values, TopLeft=@n7.RGB, TopRight=@n8.RGB, BottomRight=@n9.RGB, BottomLeft=@n10.RGB, InsetBaseMask, InsetInnerMask) {materialFunction=&/MF_SDPoly}.BaseMask
    B=MaterialFunctionCall(UVs=@n13, TopLeft=@n7.RGB, TopRight=@n8.RGB, BottomRight=@n9.RGB, BottomLeft=@n10.RGB) {materialFunction=&MF_SDPoly}.BaseMask
   ))
 ]
}
