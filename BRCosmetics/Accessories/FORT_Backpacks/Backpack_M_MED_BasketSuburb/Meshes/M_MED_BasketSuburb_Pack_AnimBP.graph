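# Animation Blueprint for the BasketSuburb backpack cosmetic, animating the pack in response to its wearer.
# A front-end flag selects between a chest-open reaction machine and a gameplay machine covering idle, run, stop, DBNO, damage and chest reactions.
# The blended pose feeds spring AnimDynamics on the pack bone, weakened during skydive floating, then root-joint offsets.
# Anim input events pulse chest and damage flags; the treasure-chest dynamics node is unconnected.
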AnimBlueprint /BRCosmetics/Accessories/FORT_Backpacks/Backpack_M_MED_BasketSuburb/Meshes/M_MED_BasketSuburb_Pack_AnimBP
target_skeleton=&/M_MED_BasketSuburb_Pack_Skeleton
graphs [
 AnimGraph {nodes [
   Anim Dynamics - Bone: dyn_jnt_treasure_chest_main(Alpha=1, ExternalForce="0,0,0") {
    bAngularSpring=true
    bLinearSpring=true
    bOverrideLinearDamping=true
    bUseGravityOverride=true
    gravityScale=0.35
    linearDampingOverride=0.999
    linearSpringConstant=90
    alphaBoolBlend {blendOption=Linear}
    alphaScaleBiasClamp {scale=0.73}
    boundBone {boneName=dyn_jnt_treasure_chest_main}
    componentLinearAccScale=<0.3 0.3 0.3>
    componentLinearVelScale=<0.3 0.3 0.3>
    gravityOverride {z=-600}
    physicsBodyDefinitions [{
      sphereCollisionRadius=0
      boundBone {boneName=dyn_jnt_treasure_chest_main}
      boxExtents=<90 100 100>
      constraintSetup {angularLimitsMax=<0 0 0>, angularLimitsMin=<0 0 0>, angularTarget=<1 0 0>, linearAxesMax=<2 0 0>}
      localJointOffset=<1 0 0>
     }]
    relativeSpaceBone {boneName=root}
   }
   Input Pose()
   AnimGraphNode_StateMachine_0 = BasketSuburb_StateMachine() {is=title AnimGraphNode_StateMachine}
   AnimGraphNode_StateMachine_1 = FE_Reactivity() {is=title AnimGraphNode_StateMachine}
   n1 = Get bIsInFrontEnd()
   n2 = Anim Dynamics - Bone: pack(
    ComponentPose=Local To Component(LocalPose=Blend Poses by bool(BlendPose_0=@AnimGraphNode_StateMachine_1, BlendPose_1=@AnimGraphNode_StateMachine_0, bActiveValue=@n1) {blendType=Linear})
    Alpha=Get skydiveDampen() -> (skydiveDampen:double)
    ExternalForce="0,0,0"
   ) {
    bLinearSpring=true
    bOverrideLinearDamping=true
    bUseGravityOverride=true
    gravityScale=0.35
    linearDampingOverride=0.999
    linearSpringConstant=90
    alphaBoolBlend {blendOption=Linear}
    boundBone {boneName=pack}
    componentLinearAccScale=<0.25 0.25 0.25>
    componentLinearVelScale=<0.25 0.25 0.25>
    gravityOverride {z=-500}
    physicsBodyDefinitions [{
      sphereCollisionRadius=0
      boundBone {boneName=pack}
      boxExtents=<30 40 40>
      constraintSetup {angularLimitsMax=<0 0 0>, angularLimitsMin=<0 0 0>, angularTarget=<1 0 0>, linearAxesMax=<0 0 2>, linearAxesMin=<0 0 -1>}
      localJointOffset=<1 0 0>
     }]
    relativeSpaceBone {boneName=root}
   }
   n3 = "Transform (Modify) Bone - Bone: jnt_Off_Root_m"!(ComponentPose=@n2, bAlphaBoolEnabled:bool=True, Translation="-5,-10,0", Rotation="0,0,0", Scale="1,1,1") -> (Pose) {
    alphaInputType=Bool
    rotationMode=BMM_Additive
    rotationSpace=BCS_BoneSpace
    translationMode=BMM_Additive
    translationSpace=BCS_BoneSpace
    alphaBoolBlend {blendInTime=0.2, blendOption=Linear, blendOutTime=0.2}
    boneToModify {boneName=jnt_Off_Root_m}
    translation=<-5 -10>
   }
   n4 = "Transform (Modify) Bone - Bone: jnt_Off_Root_m"!(ComponentPose=@n3, Alpha:float=1, Translation="0,0,0", Rotation="0,0,0", Scale="1,1,1") -> (Pose) {
    rotationMode=BMM_Additive
    rotationSpace=BCS_BoneSpace
    alphaBoolBlend {blendOption=Linear}
    alphaScaleBiasClamp {bInterpResult=true}
    boneToModify {boneName=jnt_Off_Root_m}
   }
   Output Pose(Result=Component To Local(ComponentPose=@n4)) {layerGroup=None}
  ]}
 BasketSuburb_StateMachine {
  entry=Idle
  path=AnimGraph.AnimGraphNode_StateMachine_0.BasketSuburb_StateMachine
  ids=id
  nodes [
   AnimStateNode_0 = Idle() {
    stateEntered {guid=#1}
    stateFullyBlended {guid=#2}
    stateLeft {guid=#3}
   }
   AnimStateNode_1 = Run() {is=title AnimStateNode}
   AnimStateNode_2 = Stop() {
    is=title AnimStateNode
    stateEntered {guid=#4}
    stateFullyBlended {guid=#5}
    stateLeft {guid=#6}
   }
   AnimStateNode_3 = Reactive Damage() {
    is=title AnimStateNode
    stateEntered {guid=#7}
    stateFullyBlended {guid=#8}
    stateLeft {guid=#9}
   }
   AnimStateNode_4 = ChestReaction() {is=title AnimStateNode}
   AnimStateNode_5 = DBNO() {is=title AnimStateNode}
   AnimStateTransitionNode_0 = AnimStateTransitionNode() {crossfadeDuration=0.1}
   AnimStateTransitionNode_1 = AnimStateTransitionNode() {bAutomaticRuleBasedOnSequencePlayerInState=true, crossfadeDuration=0}
   AnimStateTransitionNode_11 = AnimStateTransitionNode() {bAutomaticRuleBasedOnSequencePlayerInState=true, crossfadeDuration=0}
   AnimStateTransitionNode_12 = AnimStateTransitionNode() {crossfadeDuration=0}
   AnimStateTransitionNode_2 = AnimStateTransitionNode() {
    crossfadeDuration=0
    transitionEnd {guid=#10}
    transitionInterrupt {guid=#11}
    transitionStart {guid=#12}
   }
   AnimStateTransitionNode_3 = AnimStateTransitionNode() {crossfadeDuration=0.1}
   AnimStateTransitionNode_4 = AnimStateTransitionNode() {crossfadeDuration=0.1}
   AnimStateTransitionNode_5 = AnimStateTransitionNode() {crossfadeDuration=0}
   AnimStateTransitionNode_6 = AnimStateTransitionNode() {bAutomaticRuleBasedOnSequencePlayerInState=true}
   AnimStateTransitionNode_7 = AnimStateTransitionNode() {
    crossfadeDuration=0.1
    transitionEnd {guid=#13}
    transitionInterrupt {guid=#14}
    transitionStart {guid=#15}
   }
   AnimStateTransitionNode_8 = AnimStateTransitionNode() {crossfadeDuration=0}
  ]
  links [
   {
    dst=Idle
    rule=AnimGraph.AnimGraphNode_StateMachine_0.BasketSuburb_StateMachine.AnimStateTransitionNode_1.Transition
    src=ChestReaction
    via=AnimStateTransitionNode_1
   }
   {
    dst=Idle
    rule=AnimGraph.AnimGraphNode_StateMachine_0.BasketSuburb_StateMachine.AnimStateTransitionNode_4.Transition
    src=DBNO
    via=AnimStateTransitionNode_4
   }
   {
    dst=ChestReaction
    rule=AnimGraph.AnimGraphNode_StateMachine_0.BasketSuburb_StateMachine.AnimStateTransitionNode_12.Transition
    src=Idle
    via=AnimStateTransitionNode_12
   }
   {
    dst=DBNO
    rule=AnimGraph.AnimGraphNode_StateMachine_0.BasketSuburb_StateMachine.AnimStateTransitionNode_3.Transition
    src=Idle
    via=AnimStateTransitionNode_3
   }
   {
    dst=Reactive Damage
    rule=AnimGraph.AnimGraphNode_StateMachine_0.BasketSuburb_StateMachine.AnimStateTransitionNode_5.Transition
    src=Idle
    via=AnimStateTransitionNode_5
   }
   {
    dst=Run
    rule=AnimGraph.AnimGraphNode_StateMachine_0.BasketSuburb_StateMachine.AnimStateTransitionNode_0.Transition
    src=Idle
    via=AnimStateTransitionNode_0
   }
   {
    dst=Idle
    rule=AnimGraph.AnimGraphNode_StateMachine_0.BasketSuburb_StateMachine.AnimStateTransitionNode_6.Transition
    src=Reactive Damage
    via=AnimStateTransitionNode_6
   }
   {
    dst=Stop
    rule=AnimGraph.AnimGraphNode_StateMachine_0.BasketSuburb_StateMachine.AnimStateTransitionNode_7.AnimationTransitionGraph_1
    src=Run
    via=AnimStateTransitionNode_7
   }
   {
    dst=Stop
    rule=AnimGraph.AnimGraphNode_StateMachine_0.BasketSuburb_StateMachine.AnimStateTransitionNode_8.Transition
    src=Run
    via=AnimStateTransitionNode_8
   }
   {
    dst=Idle
    rule=AnimGraph.AnimGraphNode_StateMachine_0.BasketSuburb_StateMachine.AnimStateTransitionNode_11.Transition
    src=Stop
    via=AnimStateTransitionNode_11
   }
   {
    dst=Idle
    rule=AnimGraph.AnimGraphNode_StateMachine_0.BasketSuburb_StateMachine.AnimStateTransitionNode_2.AnimationTransitionGraph_0
    src=Stop
    via=AnimStateTransitionNode_2
   }
  ]
 }
 Idle {
  path=AnimGraph.AnimGraphNode_StateMachine_0.BasketSuburb_StateMachine.AnimStateNode_0.Idle
  nodes [Output Animation Pose(Result=Sequence Player 'BasketSuburb_Idle_PACK'() {sequence=/BRCosmetics/Accessories/FORT_Backpacks/Backpack_M_MED_BasketSuburb/Animation/BasketSuburb_Idle_PACK})]
 }
 Run {
  path=AnimGraph.AnimGraphNode_StateMachine_0.BasketSuburb_StateMachine.AnimStateNode_1.Run
  nodes [
   Blend Poses by bool(
    BlendPose_1=Sequence Player 'BasketSuburb_Run_PACK'() {sequence=/BRCosmetics/Accessories/FORT_Backpacks/Backpack_M_MED_BasketSuburb/Animation/BasketSuburb_Run_PACK}
    BlendTime_0=0
    BlendTime_1=0
    bActiveValue=Get bIsCrouching()
   ) {blendType=Linear}
   n1 = Sequence Player 'BasketSuburb_Run_PACK'() {
    groupName=Run
    method=SyncGroup
    sequence=&BasketSuburb_Run_PACK
   }
   Output Animation Pose(Result=@n1)
  ]
 }
 Stop {
  path=AnimGraph.AnimGraphNode_StateMachine_0.BasketSuburb_StateMachine.AnimStateNode_2.Stop
  nodes [Output Animation Pose(Result=Sequence Player 'BasketSuburb_Stop_PACK'!(bLoopAnimation:bool=False) -> (Pose) {bLoopAnimation=false, sequence=/BRCosmetics/Accessories/FORT_Backpacks/Backpack_M_MED_BasketSuburb/Animation/BasketSuburb_Stop_PACK})]
 }
 Reactive Damage {
  path=AnimGraph.AnimGraphNode_StateMachine_0.BasketSuburb_StateMachine.AnimStateNode_3.Reactive Damage
  nodes [
   Sequence Player 'GallonBag_Reactive_Damage'() {bLoopAnimation=false, sequence=/BRCosmetics/Accessories/FORT_Backpacks/Backpack_M_MED_GallonBag/Animation/GallonBag_Reactive_Damage}
   n1 = Sequence Player 'BasketSuburb_TakeDamage_PACK'() {bLoopAnimation=false, sequence=/BRCosmetics/Accessories/FORT_Backpacks/Backpack_M_MED_BasketSuburb/Animation/BasketSuburb_TakeDamage_PACK}
   Output Animation Pose(Result=@n1)
  ]
 }
 ChestReaction {
  path=AnimGraph.AnimGraphNode_StateMachine_0.BasketSuburb_StateMachine.AnimStateNode_4.ChestReaction
  nodes [
   n1 = Sequence Player 'BasketSuburb_ChestReactive_PACK'() {bLoopAnimation=false, sequence=/BRCosmetics/Accessories/FORT_Backpacks/Backpack_M_MED_BasketSuburb/Animation/BasketSuburb_ChestReactive_PACK}
   Output Animation Pose(Result=@n1)
  ]
 }
 DBNO {
  path=AnimGraph.AnimGraphNode_StateMachine_0.BasketSuburb_StateMachine.AnimStateNode_5.DBNO
  nodes [Output Animation Pose(Result=Sequence Player 'BasketSuburb_DBNO_PACK'!(bLoopAnimation:bool=False) -> (Pose) {bLoopAnimation=false, sequence=/BRCosmetics/Accessories/FORT_Backpacks/Backpack_M_MED_BasketSuburb/Animation/BasketSuburb_DBNO_PACK})]
 }
 Transition {
  path=AnimGraph.AnimGraphNode_StateMachine_0.BasketSuburb_StateMachine.AnimStateTransitionNode_0.Transition
  nodes [Result(bCanEnterTransition=Get bIsMoving2D())]
 }
 Transition {path=AnimGraph.AnimGraphNode_StateMachine_0.BasketSuburb_StateMachine.AnimStateTransitionNode_1.Transition, nodes [Result()]}
 Transition {path=AnimGraph.AnimGraphNode_StateMachine_0.BasketSuburb_StateMachine.AnimStateTransitionNode_11.Transition, nodes [Result()]}
 Transition {
  path=AnimGraph.AnimGraphNode_StateMachine_0.BasketSuburb_StateMachine.AnimStateTransitionNode_12.Transition
  nodes [Result(bCanEnterTransition=Get bChestOpened() -> (bChestOpened:bool))]
 }
 AnimationTransitionGraph_0 {
  path=AnimGraph.AnimGraphNode_StateMachine_0.BasketSuburb_StateMachine.AnimStateTransitionNode_2.AnimationTransitionGraph_0
  nodes [Result(bCanEnterTransition=Get bIsDBNO())]
 }
 Transition {
  path=AnimGraph.AnimGraphNode_StateMachine_0.BasketSuburb_StateMachine.AnimStateTransitionNode_3.Transition
  nodes [Result(bCanEnterTransition=Get bIsDBNO())]
 }
 Transition {
  path=AnimGraph.AnimGraphNode_StateMachine_0.BasketSuburb_StateMachine.AnimStateTransitionNode_4.Transition
  nodes [Result(bCanEnterTransition=NOT Boolean(A=Get bIsDBNO()))]
 }
 Transition {
  path=AnimGraph.AnimGraphNode_StateMachine_0.BasketSuburb_StateMachine.AnimStateTransitionNode_5.Transition
  nodes [Result(bCanEnterTransition=Get Reactive Anim Damage() -> (Reactive Anim Damage:bool))]
 }
 Transition {path=AnimGraph.AnimGraphNode_StateMachine_0.BasketSuburb_StateMachine.AnimStateTransitionNode_6.Transition, nodes [Result()]}
 AnimationTransitionGraph_1 {
  path=AnimGraph.AnimGraphNode_StateMachine_0.BasketSuburb_StateMachine.AnimStateTransitionNode_7.AnimationTransitionGraph_1
  nodes [Result(bCanEnterTransition=NOT Boolean(A=Get bIsMoving2D()))]
 }
 Transition {
  path=AnimGraph.AnimGraphNode_StateMachine_0.BasketSuburb_StateMachine.AnimStateTransitionNode_8.Transition
  nodes [Result(bCanEnterTransition=Get bIsDBNO())]
 }
 FE_Reactivity {
  entry=Idle
  path=AnimGraph.AnimGraphNode_StateMachine_1.FE_Reactivity
  ids=id
  nodes [
   AnimStateNode_0 = Idle()
   AnimStateNode_2 = FE_ChestOpen() {is=title AnimStateNode, stateEntered {guid=#16, notifyName=Entered FE_Reactive_Anim}}
   AnimStateTransitionNode_8 = AnimStateTransitionNode()
   AnimStateTransitionNode_9 = AnimStateTransitionNode() {bAutomaticRuleBasedOnSequencePlayerInState=true}
  ]
  links [
   {
    dst=Idle
    rule=AnimGraph.AnimGraphNode_StateMachine_1.FE_Reactivity.AnimStateTransitionNode_9.Transition
    src=FE_ChestOpen
    via=AnimStateTransitionNode_9
   }
   {
    dst=FE_ChestOpen
    rule=AnimGraph.AnimGraphNode_StateMachine_1.FE_Reactivity.AnimStateTransitionNode_8.Transition
    src=Idle
    via=AnimStateTransitionNode_8
   }
  ]
 }
 Idle {
  path=AnimGraph.AnimGraphNode_StateMachine_1.FE_Reactivity.AnimStateNode_0.Idle
  nodes [Output Animation Pose(Result=Sequence Player 'BasketSuburb_Idle_PACK'() {bLoopAnimation=false, sequence=&BasketSuburb_Idle_PACK})]
 }
 FE_ChestOpen {
  path=AnimGraph.AnimGraphNode_StateMachine_1.FE_Reactivity.AnimStateNode_2.FE_ChestOpen
  nodes [
   n1 = Sequence Player 'BasketSuburb_ChestReactive_PACK'() {bLoopAnimation=false, sequence=&BasketSuburb_ChestReactive_PACK}
   Output Animation Pose(Result=@n1)
  ]
 }
 Transition {
  path=AnimGraph.AnimGraphNode_StateMachine_1.FE_Reactivity.AnimStateTransitionNode_8.Transition
  nodes [Result(bCanEnterTransition=Get Play Reactive FE Anim() -> (Play Reactive FE Anim:bool))]
 }
 Transition {path=AnimGraph.AnimGraphNode_StateMachine_1.FE_Reactivity.AnimStateTransitionNode_9.Transition, nodes [Result()]}
 EventGraph {nodes [
   Event BlueprintUpdateAnimation()
   > n1 = Sequence()
   > n2 = Branch(Condition=Get bIsSkydiveFloating())
   > Set skydiveDampen(skydiveDampen:double=0.4) -> (Output_Get:double)
   Set skydiveDampen(execute=@n2.else, skydiveDampen=1)
   n3 = Branch(execute=@n1.then_1, Condition=Get bIsDBNO())
   > Set isDBNO(isDBNO:double=1) -> (Output_Get:double)
   > Set isDBNOoff(isDBNOoff:double=0) -> (Output_Get:double)
   Set isDBNOoff(execute=@n3.else, isDBNOoff=1)
   > Set isDBNO(isDBNO=0)
   Event OnCharacterCustomizationLoaded()
   > n4 = Cast To FortPlayerPawn(Object=TryGetPawnOwner())
   > Cast To B_Athena_PartModifier_Backpack_BasketSuburb!(
    execute
    Object=GetModifierForPartType!(self:Fort Player Pawn Object Reference=@n4, PartType:EFortCustomPartType Enum=Backpack) -> (ReturnValue:Custom Character Part Modifier Object Reference)
   ) -> (then, CastFailed, AsB Athena Part Modifier Backpack Basket Suburb:B Athena Part Modifier Backpack Basket Suburb Object Reference) {is=title K2Node_DynamicCast}
   > n5 = Branch()
   > Do Once(Reset=@n7.then)
   > n6 = Delay(Duration=3)
   Set Play Reactive FE Anim(execute=[@n8.then, @n6.then], Play Reactive FE Anim:bool=true) -> (Output_Get:bool)
   n7 = Set Play Reactive FE Anim(execute=@n5.else, Play Reactive FE Anim=false)
   AnimNotify_FE_ReactiveAnim() {is=title K2Node_Event}
   > Set Play Reactive FE Anim(Play Reactive FE Anim=false)
   > n8 = Delay(Duration=10)
   Event BlueprintPostEvaluateAnimation() {is=title K2Node_Event}
   > Set bChestOpened(bChestOpened:bool=false) -> (Output_Get:bool)
   > Set Play Reactive FE Anim(Play Reactive FE Anim=false)
   n9 = Event BlueprintProcessAnimInputEvent()
   > n10 = Sequence()
   > Branch(Condition=IsNamed(self=@n9.AnimInputEvent, InEventName=aie.ChestOpened))
   > Set bChestOpened(bChestOpened=true)
   > RetriggerableDelay(Duration=0.2)
   > Set bChestOpened(bChestOpened=false)
   Branch(
    execute=@n10.then_1
    Condition=OR Boolean(
     A=IsNamed(self=@n9.AnimInputEvent, InEventName=aie.PlayerVitality.ShieldsDecreased)
     B=IsNamed(self=@n9.AnimInputEvent, InEventName=aie.PlayerVitality.HealthDecreased)
    )
   )
   > Set Reactive Anim Damage(Reactive Anim Damage:bool=true) -> (Output_Get:bool)
   > RetriggerableDelay(Duration=0.2)
   > Set Reactive Anim Damage(Reactive Anim Damage=false)
   Reroute Node()
   > n11 = Branch(Condition=Get bIsDBNO())
   n12 = GetOwnerSkeletalMeshComponent()
   n13 = BreakRotator(InRot=BreakTransform(InTransform=GetDeltaTransformFromRefPose(self=@n12, BoneName=spine_05, BaseName=root)).Rotation)
   n14 = ClampAxis!(Angle:float=@n13.Yaw) -> (ReturnValue:float)
   > Set Gimbal Rot(Gimbal Rot:Rotator=MakeRotator(Roll=float * float(A=@n14, B=-1) {is=title K2Node_PromotableOperator})) -> (Output_Get:Rotator)
   n15 = MakeRotator(Roll=float * float(A=@n13.Yaw, B=-1), Pitch=float * float(A=@n13.Yaw, B=0.5))
   Set Gimbal Rot(execute=@n11.else, Gimbal Rot=@n15)
   > n16 = Branch(Condition=Get bIsSkydiveDiveMode())
   > Set Gimbal Rot(Gimbal Rot=@n15)
   Set ADSGimbal(execute=@n16.else, ADSGimbal:Rotator=MakeRotator(Roll=float * float(A=@n14, B=-1), Pitch=float * float(A=@n14, B=0.5))) -> (Output_Get:Rotator)
   n17 = Get bIsSkydiveDiveMode()
   > Set AlphaGimbal(AlphaGimbal:double="To Float (Boolean)"(InBool=NOT Boolean(A=@n17))) -> (Output_Get:double)
   TryGetPawnOwner()
   NOT Boolean(A=Get bIsCrouching())
   Get FE_ReactivityEnabled!(self:B Athena Part Modifier Backpack Twinkle Loop Object Reference) -> (FE_ReactivityEnabled:bool)
   Set offsetBP(offsetBP:bool=OR Boolean(A=Get bIsOnGrindRail(), B=Get bIsSkyDiving())) -> (Output_Get:bool)
  ]}
]
variables [
 ADSGimbal {type=Rotator}
 AlphaGimbal {type=number}
 As B Athena Part Modifier Backpack Basket Suburb {type=B_Athena_PartModifier_Backpack_BasketSuburb_C, default=null}
 bChestOpened {type=boolean}
 FrontEnd {type=boolean}
 Gimbal Rot {type=Rotator}
 isDBNO {type=number}
 isDBNOoff {type=number}
 isHappy {type=boolean}
 offsetBP {type=boolean}
 Play Reactive FE Anim {type=boolean}
 PlayerPawn {type=FortPlayerPawn}
 Reactive Anim Damage {type=boolean}
 Reactivity Enabled {type=boolean}
 skydiveDampen {type=number}
 YawDeltaPositive {type=boolean}
 YawDeltaPositiveLastTick {type=boolean}
]
